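; Child Blueprint of the Icy Grappler hook projectile that adds a camera lens effect when a grapple occurs.
; EventGrapple runs the parent logic, then, where cosmetic events can execute, gets the player pawn.
; If that pawn lookup succeeds, it spawns a generic camera lens effect on it, using the class from the lens effect interface class variable.
; The construction script only calls the parent.
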(event EventGrapple
  ; Make sure to applay the camera lens effect.
  (|CallParentFunction)
  (Networking|CanExecuteCosmeticEvents
    (:True
      (bind _playerpawn (Variables|Default|GetPlayerPawn)
        (:then
          (Feedback|SpawnGenericCameraLensEffect _playerpawn (LensEffect|GetInterfaceClass (Variables|Default|GetLensEffectInterfaceClass))))
        (:else)))))

(fn ConstructionScript
  (|CallParentFunction))
