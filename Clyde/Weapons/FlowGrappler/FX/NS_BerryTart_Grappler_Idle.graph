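# Niagara idle effect for the BerryTart skin of the FlowGrappler weapon, emitting short twinkling sparkles over the grappler.
# Sparkles_Back fires bursts on randomized loops and places sprites on a sparkle-location static mesh, pulsing size and alpha.
# View-dependent logic kills particles by XZ-flattened normal-to-camera dot product and varies camera offset with viewing angle along Y.
# A disabled Debugger emitter writes a facing dot product to a material parameter.
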NiagaraSystem /Clyde/Weapons/FlowGrappler/FX/NS_BerryTart_Grappler_Idle
dependencies {
 usedDataInterfaces [&NiagaraDataInterfaceCurve, &NiagaraDataInterfaceCamera, &NiagaraDataInterfaceStaticMesh]
 usedDynamicInputs []
 usedModules [&SystemState]
}
emitters {
 Debugger {
  bEnabled=false
  emitterUpdateScript [
   &EmitterState()
   &SpawnBurst_Instantaneous(Use Spawn Probability=false, UseLoopCountLimit=false)
  ]
  particleSpawnScript [
   &InitializeParticle(Color Mode=::NewEnumerator0, Lifetime=5, Position Offset=<2.5 25 14.5>, UsePositionOffset=true, Write Material Random=true)
   &CameraOffset(Camera Offset Amount=35)
  ]
  particleUpdateScript [
   &ParticleState()
   /Niagara/Modules/Spawn/Location/StaticMeshLocation(
    MaskSpawnGroup=false
    Offset Position Along Sampled Normal=0.1
    OffsetAlongNormal=true
    OverrideRandomSeed=false
    SamplePosOffset=false
    Static Mesh=DataInterface() {propertyValues {
      DefaultMesh=&/SM_BerryTart_Grappler_SpakleLocation
      InstanceIndex=-1
      LODIndex=0
      PreviewMesh=&SM_BerryTart_Grappler_SpakleLocation
      SoftSourceActor=null
      SourceMode=Default
      bAllowSamplingFromStreamingLODs=false
      bCaptureTransformsPerFrame=true
      bUsePhysicsBodyVelocity=false
      FilteredSockets []
      LODIndexUserParameter {parameter {name=None, typeDefHandle {registeredTypeIndex=108}, varData []}}
      MeshParameterBinding {parameter {name=None, typeDefHandle {registeredTypeIndex=1}, varData []}}
      SectionFilter {allowedMaterialSlots []}
     }}
   ) {enabled=false}
   &DynamicMaterialParameters(
    Index 0 Param 1=&DotProduct(
     Vector A=$Particles.StaticMeshLocation.SampledNormal
     Vector B=&NormalizeVector(
      Vector To Normalize=&ReturnCameraVector(Camera Vectors=::NewEnumerator4, Coordinate Space=::Simulation, Origin Position=$Engine.Owner.Position)
      Safe Normalize=true
      Safe Fallback Threshold=0
      Safe Fallback Vector=<0 0 0.001>
     )
    )
    Module.Index0Param0Write=true
    Module.Index0Param1Write=true
    Module.Index0Param2Write=true
    Module.Index0Param3Write=true
   )
  ]
  rendererClasses [&NiagaraSpriteRendererProperties, &NiagaraMeshRendererProperties]
  renderers [
   {rendererClass=&NiagaraSpriteRendererProperties}
   {rendererClass=&NiagaraMeshRendererProperties, rendererIndex=1}
  ]
 }
 Sparkles_Back {
  emitterUpdateScript [
   &EmitterState(
    Life Cycle Mode=::NewEnumerator1
    Loop Duration=&Lerp_Float(
     A=0.07
     B=0.7
     Alpha=&FloatFromCurve(
      FloatCurve=DataInterface() {propertyValues {Curve {keys [
          {
           arriveTangent=5.105
           arriveTangentWeight=0.907
           interpMode=RCIM_Cubic
           leaveTangent=5.105
           leaveTangentWeight=0.907
           tangentMode=RCTM_User
           tangentWeightMode=RCTWM_WeightedBoth
           time=0
           value=0
          }
          {
           arriveTangentWeight=0.5828
           interpMode=RCIM_Cubic
           leaveTangentWeight=0.5828
           tangentMode=RCTM_User
           tangentWeightMode=RCTWM_WeightedBoth
           time=1
           value=1
          }
         ]}}}
      CurveIndex=&RandomRangeFloat(Minimum=0, Maximum=1)
     )
    )
    Recalculate Duration Each Loop=true
    UseLoopDelay=false
   )
   &SpawnBurst_Instantaneous(Use Spawn Probability=false, UseLoopCountLimit=false)
   &SpawnBurst_Instantaneous(Spawn Time=1.0833, Use Spawn Probability=false, UseLoopCountLimit=false) {module=SpawnBurst_Instantaneous001}
  ]
  particleSpawnScript [&InitializeParticle(
    Color=<rgba 1 1.5623 2 2>
    Lifetime Max=1
    Lifetime Min=1.35
    Lifetime Mode=::NewEnumerator1
    Mass Min=0.675
    Mass Mode=::NewEnumerator3
    Sprite Rotation Angle=-25
    Sprite Rotation Mode=::NewEnumerator0
    Sprite Size Mode=::NewEnumerator1
    Uniform Sprite Size Max=20
    Uniform Sprite Size Min=15
    UsePositionOffset=false
    Write Material Random=true
   )]
  particleUpdateScript [
   &ParticleState()
   /SetVariables(Particles.VectorToCam:Vector3f=&ReturnCameraVector(Camera Vectors=::NewEnumerator4, Coordinate Space=::Simulation, Origin Position=$Particles.Position)) {isSetParameters=true}
   &StaticMeshLocation(
    MaskSpawnGroup=false
    Offset Position Along Sampled Normal=0.1
    OffsetAlongNormal=true
    OverrideRandomSeed=false
    SamplePosOffset=false
    Static Mesh=DataInterface() {propertyValues {
      DefaultMesh=&SM_BerryTart_Grappler_SpakleLocation
      InstanceIndex=-1
      LODIndex=0
      PreviewMesh=&SM_BerryTart_Grappler_SpakleLocation
      SoftSourceActor=null
      SourceMode=Default
      bAllowSamplingFromStreamingLODs=false
      bCaptureTransformsPerFrame=true
      bUsePhysicsBodyVelocity=false
      FilteredSockets []
      LODIndexUserParameter {parameter {name=None, typeDefHandle {registeredTypeIndex=108}, varData []}}
      MeshParameterBinding {parameter {name=None, typeDefHandle {registeredTypeIndex=1}, varData []}}
      SectionFilter {allowedMaterialSlots []}
     }}
   )
   &KillParticles(Kill Particles=&SetBoolByFloatComparison(
     A=&FloatFromCurve(
      FloatCurve=DataInterface() {propertyValues {Curve {keys [<at -1 0 RCIM_Constant>,<at -0.35 1 RCIM_Constant>,<at 0.9 0 RCIM_Constant>]}}}
      CurveIndex=&DotProduct(
       Vector A=&NormalizeVector(
        Vector To Normalize=&Multiply_Vector(A=$Particles.VectorToCam, B=<1 0 1>)
        Safe Normalize=true
        Safe Fallback Threshold=0
        Safe Fallback Vector=<0 0 0.001>
       )
       Vector B=&NormalizeVector(
        Vector To Normalize=&Multiply_Vector(A=$Particles.StaticMeshLocation.SampledNormal, B=<1 0 1>)
        Safe Normalize=true
        Safe Fallback Threshold=0
        Safe Fallback Vector=<0 0 0.001>
       )
      )
     )
     B=0.5
     Comparison Type=::NewEnumerator7
    ))
   &CameraOffset(Camera Offset Amount=&Lerp_Float(
     A=75
     B=1
     Alpha=&FloatFromCurve(
      FloatCurve=DataInterface() {propertyValues {Curve {keys [
          {
           arriveTangent=6.3324
           arriveTangentWeight=0.6002
           interpMode=RCIM_Cubic
           leaveTangent=6.3324
           leaveTangentWeight=0.6002
           tangentMode=RCTM_User
           tangentWeightMode=RCTWM_WeightedBoth
           time=0
           value=0
          }
          {
           arriveTangentWeight=0.9499
           interpMode=RCIM_Cubic
           leaveTangentWeight=0.9499
           tangentMode=RCTM_User
           tangentWeightMode=RCTWM_WeightedBoth
           time=1
           value=1
          }
         ]}}}
      CurveIndex=&Abs_Float(Float:NiagaraFloat=&DotProduct(
        Vector A=<0 1 0>
        Vector B=&NormalizeVector(
         Vector To Normalize=&ReturnCameraVector(Camera Vectors=::NewEnumerator4, Coordinate Space=::Simulation, Origin Position=$Engine.Owner.Position)
         Safe Normalize=true
         Safe Fallback Threshold=0
         Safe Fallback Vector=<0 0 0.001>
        )
       ))
     )
    ))
   &ScaleColor(
    Scale Alpha=&FloatFromCurve(
     FloatCurve=DataInterface() {propertyValues {Curve {keys [
         {arriveTangent=5.883, interpMode=RCIM_Cubic, leaveTangent=5.883, tangentMode=RCTM_User, time=0, value=0}
         <at 0.5 1 RCIM_Cubic>
         {arriveTangent=-5.491, interpMode=RCIM_Cubic, leaveTangent=-5.491, tangentMode=RCTM_User, time=1, value=0}
        ]}}}
     Scale Curve=3
    )
    ScaleA=true
    ScaleRGB=true
   ) {module=ScaleColor001}
   &ScaleSpriteSize(Uniform Curve Sprite Scale=DataInterface() {propertyValues {Curve {keys [
        {interpMode=RCIM_Cubic, leaveTangentWeight=0.8281, tangentMode=RCTM_User, time=0, value=0}
        {arriveTangent=5.598, interpMode=RCIM_Cubic, leaveTangent=-5.1027, tangentMode=RCTM_Break, time=0.5, value=1}
        {interpMode=RCIM_Cubic, leaveTangent=0.1974, time=1, value=0}
       ]}}})
  ]
  rendererClasses [&NiagaraSpriteRendererProperties]
  renderers [
   {rendererClass=&NiagaraSpriteRendererProperties}
   {rendererClass=&NiagaraSpriteRendererProperties, rendererIndex=1}
  ]
 }
}
summary {
 systemName=NS_BerryTart_Grappler_Idle
 emitters [
  {emitterName=Sparkles_Back, rendererClasses [&NiagaraSpriteRendererProperties]}
  {
   bEnabled=false
   emitterName=Debugger
   rendererClasses [&NiagaraSpriteRendererProperties, &NiagaraMeshRendererProperties]
  }
 ]
 userVariables []
}
