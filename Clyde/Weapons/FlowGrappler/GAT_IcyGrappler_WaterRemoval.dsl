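; Gameplay ability for the Icy Grappler that strips icy feet effects from its owner, apparently as the water-removal counterpart to that status.
; On activation it calls the parent, removes owner effects granting the grappler, ice grenade or ice trap IcyFeet tags, then ends immediately.
; The graph contains no water check or trigger condition.
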(event Ability|EventActivateAbility
  (|CallParentFunction)
  (Ability|RemoveGameplayEffectFromOwnerWithGrantedTags "(GameplayTags=((TagName=\"GameplayEffect.Item.IceGrenade.IcyFeet\"),(TagName=\"GameplayEffect.Trap.Ice.IcyFeet\"),(TagName=\"Gameplay.Status.IcyFeet.Grappler\")))")
  (Ability|EndAbility))
