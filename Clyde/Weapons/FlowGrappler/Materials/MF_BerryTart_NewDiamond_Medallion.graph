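# Material function that gives the BerryTart FlowGrappler medallion a faked diamond look, outputting modified material attributes, emissive and a diamond mask.
# Skylight reflection and refraction samples, using triplanar virtual-box or bump-offset normals, build an emissive gem interior scaled by time of day.
# A custom HLSL height ray march with cheaper platform fallbacks, Fresnel caustics, fake LUT and backlight scattering terms add depth.
# Base color, metallic, specular and normal overrides are lerped through masks.
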MaterialFunction /Clyde/Weapons/FlowGrappler/Materials/MF_BerryTart_NewDiamond_Medallion
graph {nodes [
  AtmosphericLightVector()
  Custom() {code="return View.DirectionalLightDirection;", description=DirectionalLightVector, inputs []}
  n1 = MaterialFunctionCall(Value_Morning=Constant() {r=0.5}, Value_Day=Constant() {r=1}, Value_Evening=Constant() {r=0.3}, Value_Night=Constant() {r=0.25}) {materialFunction=&Scalar_Time_of_Day}
  n2 = VectorParameter() {group=Diamond_Base, parameterName=DiamondTint, defaultValue=<rgba 0.8771 0.9768 1 1>}
  n3 = SkyLightEnvMapSample(
   Direction=ReflectionVectorWS(CustomWorldNormal=NamedRerouteUsage())
   Roughness=ScalarParameter() {defaultValue=0.1, group=Diamond_Base, parameterName=ReflectionRoughness}
  )
  n4 = MaterialFunctionCall!(
   Ray Direction=StaticSwitchParameter(True=NamedRerouteUsage(), False=NamedRerouteUsage()) {group=Diamond_Base, parameterName=UseDoubleRefraction}
   Surface Normal=NamedRerouteUsage()
   Refractive Index Origin=NamedRerouteUsage()
   Refractive Index Target=Constant() {r=1}
  ) {materialFunction=/Engine/Functions/Engine_MaterialFunctions01/Vectors/Refract}
  n5 = SkyLightEnvMapSample(Direction=@n4.Vector, Roughness=ScalarParameter() {defaultValue=0.2, group=Diamond_Base, parameterName=RefractionRoughness})
  n6 = NamedRerouteUsage()
  n7 = Multiply(
   A=Desaturation(None=LinearInterpolate(A=@n3, B=@n5, Alpha=@n6), Fraction=Constant() {r=1})
   B=ScalarParameter() {defaultValue=1, group=Diamond_Override, parameterName=BackfaceSampleStrength}
  )
  n8 = LinearInterpolate(
   A=ScalarParameter() {defaultValue=0.05, group=Diamond_Base, parameterName=DiamondMinIntensity}
   Alpha=SmoothStep(
    Min=ScalarParameter() {group=Diamond_Base, parameterName=DiamondSmoothMin}
    Max=ScalarParameter() {defaultValue=1, group=Diamond_Base, parameterName=DiamondSmoothMax}
    Value=Saturate(None=@n7)
   )
  )
  n9 = Multiply(
   A=Multiply(A=ScalarParameter() {defaultValue=2, group=Diamond_Base, parameterName=DiamondEmissiveIntensity}, B=Multiply(A=@n2.RGB, B=@n8))
   B=NamedRerouteUsage()
  )
  n10 = NamedRerouteDeclaration(None=Multiply(A=NamedRerouteUsage(), B=@n9)) {name=Emissive_VirtualPlane, nodeColor=<rgba 0 0.2734 1 1>}
  n11 = NamedRerouteUsage()
  n12 = ScalarParameter() {defaultValue=0.748, group=Diamond_Override, parameterName=GS_CoreDarkening}
  n13 = ScalarParameter() {defaultValue=0.3, group=Diamond_Override, parameterName=GS_CoreColor}
  n14 = Multiply(
   A=Add(A=NamedRerouteUsage(), B=Multiply(A=@n10, B=@n11))
   B=Multiply(
    A=Fresnel(ExponentIn=@n12, BaseReflectFractionIn=@n13) {baseReflectFraction=0.1}
    B=ScalarParameter() {defaultValue=1.3, parameterName=GS_CoreMulti}
   )
  )
  FunctionOutput(None=NamedRerouteDeclaration(None=Multiply(A=@n1.Result, B=@n14)) {name=EM_Diamond, nodeColor=<rgba 1 0.2344 0 1>}) {outputName=Emissive, sortPriority=1}
  n15 = SetMaterialAttributes!(MaterialAttributes, BaseColor=Constant() {r=1}) {attributeSetTypes [#1]}
  n16 = GetMaterialAttributes(None=NamedRerouteDeclaration(None=FunctionInput(Preview=@n15) {bUsePreviewValueAsDefault=true, inputName=Material, inputType=FunctionInput_MaterialAttributes}) {name=Material_IN, nodeColor=<rgba 0 0.3437 1 1>}) {attributeGetTypes [#1, #2, #3]}
  n17 = VectorParameter() {group=Diamond_Override, parameterName=BaseColorSubtle_Tint, defaultValue=<rgba 1 1 1 1>}
  n18 = StaticSwitchParameter(True=VectorParameter() {group=Diamond_Override, parameterName=PureColor_BaseColor, defaultValue=<rgba 1 1 1 1>}.RGB, False=@n16.BaseColor) {group=Diamond_Override, parameterName=UsePureColorAsBaseColor}
  n19 = Multiply(
   A=Multiply(A=NamedRerouteUsage(), B=NamedRerouteDeclaration(None=@n18) {name=BaseColor_OG, nodeColor=<rgba 0.0078 0 1 1>})
   B=ScalarParameter() {defaultValue=1, group=Diamond_Override, parameterName=BaseColorMulti}
  )
  n20 = NamedRerouteDeclaration(None=@n16.Normal) {name=Normal_OG, nodeColor=<rgba 1 0.3047 0 1>}
  n21 = ScalarParameter() {group=Diamond_Override, parameterName=NormalOG_Flatten}
  n22 = NamedRerouteUsage()
  n23 = ScalarParameter() {defaultValue=0.2, group=Diamond_Override, parameterName=DiamondNormal_Flatten}
  n24 = SetMaterialAttributes!(
   MaterialAttributes=@n16.MaterialAttributes
   BaseColor=Saturate(None=Multiply(A=@n17.RGB, B=@n19))
   Metallic=ScalarParameter() {defaultValue=0.75, group=Diamond_Override, parameterName=MetalicOverride}
   EmissiveColor=Add(A=@n16.EmissiveColor, B=NamedRerouteUsage())
   Specular=Multiply(
    A=ScalarParameter() {defaultValue=0.5, group=Diamond_Override, parameterName=Spec_Override}
    B=VectorParameter() {group=Diamond_Override, parameterName=Spec_Tint, defaultValue=<rgba 1 1 1 1>}.RGB
   )
   Normal=MaterialFunctionCall(
    BaseNormal=MaterialFunctionCall(Normal=@n20, Flatness=@n21) {materialFunction=&FlattenNormal}.Result
    AdditionalNormal=MaterialFunctionCall(Normal=@n22, Flatness=@n23) {materialFunction=&FlattenNormal}.Result
   ) {materialFunction=&BlendAngleCorrectedNormals}.Result
  ) {attributeSetTypes [
    #1
    #4
    #3
    #5
    #2
   ]}
  n25 = GetMaterialAttributes(None=@n24) {attributeGetTypes [
    #1
    #4
    #3
    #5
    #2
   ]}
  n26 = GetMaterialAttributes(None=NamedRerouteUsage()) {attributeGetTypes [
    #1
    #4
    #3
    #5
    #2
   ]}
  n27 = NamedRerouteUsage()
  n28 = NamedRerouteUsage()
  n29 = SetMaterialAttributes(
   MaterialAttributes=@n25.MaterialAttributes
   BaseColor=LinearInterpolate(A=@n26.BaseColor, B=@n25.BaseColor, Alpha=@n27)
   Metallic=LinearInterpolate(A=@n26.Metallic, B=@n25.Metallic, Alpha=@n27)
   EmissiveColor=LinearInterpolate(A=@n26.EmissiveColor, B=@n25.EmissiveColor, Alpha=@n28)
   Specular=LinearInterpolate(A=@n26.Specular, B=@n25.Specular, Alpha=@n28)
   Normal=LinearInterpolate(A=@n26.Normal, B=@n25.Normal, Alpha=NamedRerouteUsage())
  ) {attributeSetTypes [
    #1
    #4
    #3
    #5
    #2
   ]}
  FunctionOutput(None=@n29) {outputName=Material}
  FunctionOutput(None=NamedRerouteUsage()) {outputName=DiamondMask, sortPriority=2}
  Multiply(A=@n9, B=NamedRerouteUsage())
  n30 = TextureCoordinate()
  n31 = ScalarParameter() {defaultValue=3, group=Diamond_EnvironmentOverride, parameterName=BackfaceTiling_Bump}
  n32 = NamedRerouteUsage()
  n33 = TextureSample(
   UVs=BumpOffset(Coordinate=Multiply(A=@n30, B=@n31), Height=ScalarParameter() {defaultValue=-5, group=Diamond_EnvironmentOverride, parameterName=BumpDepth})
   Tex=@n32
  )
  n34 = MaterialFunctionCall!(
   Texture=@n32
   Depth=ScalarParameter() {defaultValue=5, group=Diamond_Base, parameterName=BackfaceDepth}
   Tile_UVX=ScalarParameter() {defaultValue=1, group=Diamond_Base, parameterName=BackfaceTiling_X}
   Tile_UVY=ScalarParameter() {defaultValue=1, group=Diamond_Base, parameterName=BackfaceTiling_Y}
   Tile_UVZ=ScalarParameter() {defaultValue=1, group=Diamond_Base, parameterName=BackfaceTiling_Z}
   BlendContrast=ScalarParameter() {defaultValue=1, group=Diamond_Base, parameterName=BackfaceBlendContrast}
   SpeedXY
   ActorOffset=VectorParameter() {group=Diamond_Base, parameterName=BackfaceOffset, defaultValue=<rgba 0 0 50 1>}.RGB
   UseLocalCameraVector=StaticBool() {value=true}
  ) {materialFunction=&/MF_TriplanarVirtualBox_TileXYZ}
  NamedRerouteDeclaration(None=Transform(None=StaticSwitchParameter(True=@n33.RGB, False=@n34.Result) {group=Diamond_Override, parameterName=UseBumpOffsetForBackFace})) {name=BackfaceNormal, nodeColor=<rgba 0.1016 1 0 1>}
  n35 = CameraVectorWS()
  n36 = ScalarParameter() {defaultValue=1.5, group=Diamond_Base, parameterName=IOR}
  NamedRerouteDeclaration(None=MaterialFunctionCall(
    Ray Direction=Multiply(A=@n35) {constB=-1}
    Surface Normal=NamedRerouteUsage()
    Refractive Index Origin=Constant() {r=1}
    Refractive Index Target=NamedRerouteDeclaration(None=@n36) {name=IOR, nodeColor=<rgba 0 1 0.9297 1>}
   ) {materialFunction=&Refract}.Vector) {name=RefractionVector, nodeColor=<rgba 1 0 0.4219 1>}
  n37 = DeriveNormalZ(InXY=ComponentMask(None=NamedRerouteUsage()) {g=true, r=true})
  n38 = Transform(None=@n37)
  n39 = Multiply(
   A=Fresnel(ExponentIn=ScalarParameter() {defaultValue=1, group=Diamond_RayMarching, parameterName=Fresnel_Exp}, Normal=@n38) {baseReflectFraction=0, exponent=2}
   B=ScalarParameter() {defaultValue=2, group=Diamond_RayMarching, parameterName=SineMulti}
  ) {constB=2}
  n40 = NamedRerouteDeclaration(None=SmoothStep(Value=Sine(None=@n39)) {constMax=0.5, constMin=0.3}) {name=Fresnel_Caustics, nodeColor=<rgba 0.0312 0 1 1>}
  n41 = TextureSampleParameterCube!(UVs=Transform(None=@n37), Bias, Apply View MipBias) {
   group=Gemstone
   mipValueMode=TMVM_MipBias
   parameterName=GS_RefractionCubeMap
   samplerType=SAMPLERTYPE_LinearColor
   texture=/Game/Global/Textures/Cubemaps/TC_Normals_Random_S
  }
  n42 = MaterialFunctionCall(Float3=@n41.RGB) {materialFunction=&BreakOutFloat3Components}
  n43 = MaterialFunctionCall(X=@n42.R, Y=Clamp(None=@n42.G) {maxDefault=0.3}, Z=@n42.R) {materialFunction=&MakeFloat3}
  n44 = ScalarParameter() {defaultValue=-0.3, group=Diamond_RayMarching, parameterName=Desaturation_Caustics}
  n45 = Multiply(A=@n40, B=Multiply(A=Desaturation(None=@n43.Result, Fraction=@n44), B=NamedRerouteUsage()))
  n46 = ComponentMask(None=DotProduct(A=NamedRerouteUsage(), B=@n38)) {r=true}
  n47 = ScalarParameter() {defaultValue=2, group=Diamond_RayMarching, parameterName=GradientMulti}
  n48 = TextureSampleParameter2D(UVs=AppendVector(A=Multiply(A=@n46, B=@n47), B=Constant() {r=0.5})) {group=Diamond_RayMarching, parameterName=DiamondGradient, texture=&/T_BerryTart_GradientDiamond}
  n49 = Multiply(A=@n40, B=Desaturation(None=@n48.RGB, Fraction=@n44))
  NamedRerouteDeclaration(None=Multiply(
    A=ScalarParameter() {defaultValue=0.4, group=Diamond_RayMarching, parameterName=CausticsIntensity}
    B=StaticSwitchParameter(True=@n45, False=@n49) {group=Diamond_RayMarching, parameterName=UseCubeMapForCaustics}
   ) {constA=0.4}) {name=Emissive_Caustics, nodeColor=<rgba 0.5703 1 0 1>}
  NamedRerouteDeclaration(None=@n5) {name=RefractColor, nodeColor=<rgba 0.6875 1 0 1>}
  NamedRerouteDeclaration(None=@n3) {name=ReflectColor, nodeColor=<rgba 0 0.3437 1 1>}
  n50 = Custom() {code="return View.DirectionalLightDirection;", description=DirectionalLightVector, inputs []}
  n51 = DotProduct(A=@n50, B=@n50)
  n52 = CollectionParameter() {
   collection=&FortniteMaterialParameters
   parameterName=SunlightDirectionalVector
  }
  n53 = ComponentMask(None=@n52) {b=true, g=true, r=true}
  n54 = MaterialFunctionCall!(In=If(
    A=Custom() {code="return View.DirectionalLightColor;", description=DirectionalLightColor, inputs []}
    "A > B"=If(A=@n51, "A > B"=@n50, "A == B"=@n53, "A < B"=@n53) {desc=check if DirectionalLight exists}
    "A == B"=@n53
    "A < B"=@n53
   ) {desc=check if DirectionalLight color is Black}) {materialFunction=/Game/Packages/Fortress_SharedMaterials/Cinematics/Anime/MF_CINE_AnimeLightDirOverride}
  n55 = MaterialFunctionCall(Float3=@n54.Result) {materialFunction=&BreakOutFloat3Components}
  n56 = ScalarParameter() {
   desc="0.0 = the light vector is locked on the XY plane.\r\n1.0 = the actual light vector is used  \r\n(for Faces use a blend value of 0,0)"
   group=02_Anime Parameters
   parameterName=LockedZLightVector_Blend
   sliderMax=1
  }
  n57 = MaterialFunctionCall(X=@n55.R, Y=@n55.R, Z=Multiply(A=@n55.B, B=@n56)) {materialFunction=&MakeFloat3}
  n58 = StaticSwitchParameter(True=LinearInterpolate(A=@n54.Result, B=Normalize(VectorInput=@n57.Result), Alpha=VertexColor().A), False=@n54.Result) {
   desc="Lock the Z axis of the Light Vector\r\nis driven by the VertexColor A channel (0 = Regular LightVector, 1= Locked LIghtVector)"
   group=02_Anime Parameters
   parameterName=UseLockedZLightVector
   sortPriority=0
  }
  NamedRerouteDeclaration(None=@n58) {name=LightVector, nodeColor=<rgba 1 0.934 0.5103 1>}
  n59 = Multiply(A=DotProduct(A=NamedRerouteUsage(), B=NamedRerouteUsage())) {constB=-1}
  n60 = Constant() {r=0.73}
  n61 = Constant() {r=0.78}
  NamedRerouteDeclaration(None=LinearInterpolate(Alpha=MaterialFunctionCall(Input=@n59, Input Low=@n60, Input High=@n61) {materialFunction=&RemapValue_0-1}.Result) {constA=0.1, constB=0.9}) {name=TIR, nodeColor=<rgba 0 0.7422 1 1>}
  n62 = Fresnel(Normal=Transform(None=NamedRerouteUsage())) {exponent=1}
  n63 = ScalarParameter() {defaultValue=3, group=Diamond_Override, parameterName=LUTSineMulti}
  n64 = Sine(None=Add(A=Multiply(A=@n62, B=@n63) {constB=3}, B=ScalarParameter() {group=Diamond_Override, parameterName=LUTSineOffset}))
  NamedRerouteDeclaration(None=LinearInterpolate(
    A=ScalarParameter() {defaultValue=0.4, group=Diamond_Override, parameterName=LUTMin}
    B=ScalarParameter() {defaultValue=1, group=Diamond_Override, parameterName=LUTMax}
    Alpha=ConstantBiasScale(None=@n64)
   ) {constA=0.4}) {name=FakeLUT, nodeColor=<rgba 1 0.3047 0 1>}
  n65 = NamedRerouteUsage()
  n66 = Sine(None=Multiply(A=Fresnel(Normal=@n65) {exponent=1}) {constB=3})
  NamedRerouteDeclaration(None=LinearInterpolate(Alpha=ConstantBiasScale(None=@n66)) {constA=0.5}) {name=FakeLUT_Interior, nodeColor=<rgba 1 0.3047 0 1>}
  NamedRerouteDeclaration(None=@n4.Vector) {name=BackfaceRefraction, nodeColor=<rgba 0.7109 0 1 1>}
  n67 = NamedRerouteUsage()
  n68 = ScalarParameter() {defaultValue=2, group=Diamond_Base, parameterName=InteriorDepth}
  n69 = ScalarParameter() {defaultValue=2, group=Diamond_Base, parameterName=InteriorTiling_X}
  n70 = ScalarParameter() {defaultValue=2, group=Diamond_Base, parameterName=InteriorTiling_Y}
  n71 = ScalarParameter() {defaultValue=2, group=Diamond_Base, parameterName=InteriorTiling_Z}
  n72 = ScalarParameter() {defaultValue=1, group=Diamond_Base, parameterName=InteriorBlendContrast}
  n73 = VectorParameter() {group=Diamond_Base, parameterName=InteriorActorOffset, defaultValue=<rgba 0 0 37.5 1>}
  n74 = StaticBool() {value=true}
  NamedRerouteDeclaration(None=Transform(None=MaterialFunctionCall(Texture=@n67, Depth=@n68, Tile_UVX=@n69, Tile_UVY=@n70, Tile_UVZ=@n71, BlendContrast=@n72, ActorOffset=@n73.RGB, UseLocalCameraVector=@n74) {materialFunction=&MF_TriplanarVirtualBox_TileXYZ}.Result)) {name=InteriorNormal, nodeColor=<rgba 0.1016 1 0 1>}
  n75 = TextureCoordinate()
  n76 = ScalarParameter() {defaultValue=3, group=Diamond_Base, parameterName=NormalTile}
  n77 = VectorParameter() {group=Diamond_Base, parameterName=Normal_Offset_UV}
  n78 = TextureSample(UVs=Add(A=Multiply(A=@n75, B=@n76), B=ComponentMask(None=@n77.RGB) {g=true, r=true}), Tex=NamedRerouteUsage()) {samplerType=SAMPLERTYPE_Normal}
  NamedRerouteDeclaration(None=Transform(None=NamedRerouteDeclaration(None=@n78.RGB) {name=TangentNormal, nodeColor=<rgba 1 0.6094 0 1>})) {name=WorldSurfaceNormal, nodeColor=<rgba 0 1 0.1328 1>}
  NamedRerouteDeclaration(None=MaterialFunctionCall(
    Ray Direction=NamedRerouteUsage()
    Surface Normal=NamedRerouteUsage()
    Refractive Index Origin=NamedRerouteUsage()
    Refractive Index Target=Constant() {r=1}
   ) {materialFunction=&Refract}.Vector) {name=InteriorRefraction, nodeColor=<rgba 1 0 0.6562 1>}
  n79 = Constant() {r=1}
  NamedRerouteDeclaration(None=ChannelMaskParameter(None=FunctionInput(Preview=@n79) {bUsePreviewValueAsDefault=true, inputName=TechArtMask, sortPriority=1}) {group=Diamond_Base, parameterName=DiamondChannel}.R) {name=TechArtMask, nodeColor=<rgba 0.6641 1 0 1>}
  n80 = PixelNormalWS()
  n81 = ScalarParameter() {defaultValue=1, group=Diamond_Override, parameterName=BacklightViewAngle, sortPriority=16}
  n82 = CameraVectorWS()
  n83 = Normalize(VectorInput=Add(A=@n80, B=Multiply(A=@n81, B=@n82)))
  n84 = Multiply(A=SkyAtmosphereLightDirection()) {constB=-1}
  n85 = Add(A=Multiply(A=DotProduct(A=@n83, B=@n84)) {constB=0.5}) {constB=0.5}
  n86 = Multiply(
   A=Power(Base=Saturate(None=@n85), Exp=ScalarParameter() {defaultValue=2, group=Diamond_Override, parameterName=BacklightAperture, sortPriority=17})
   B=Saturate(None=DotProduct(A=@n80, B=@n82))
  )
  NamedRerouteDeclaration(None=LinearInterpolate(
    A=ScalarParameter() {defaultValue=1, group=Diamond_Override, parameterName=LightScatteringMin}
    B=ScalarParameter() {defaultValue=3, group=Diamond_Override, parameterName=LightScatteringMax}
    Alpha=Saturate(None=@n86)
   ) {constA=1, constB=1.5}) {name=Interior Light Scattering, nodeColor=<rgba 0.5 1 0 1>}
  n87 = TextureObjectParameter() {
   group=Diamond_Base
   parameterName=DiamondNormalTexture
   samplerType=SAMPLERTYPE_Normal
   texture=/ClydeCosmetics/Weapons/FlowGrappler/Materials/T_BerryTart_Diamond_N
  }
  NamedRerouteDeclaration(None=@n87) {name=NormalTexture, nodeColor=<rgba 1 0 0.75 1>}
  n88 = ScalarParameter() {defaultValue=5, group=Diamond_Base, parameterName=ExponentIn}
  n89 = NamedRerouteUsage()
  NamedRerouteDeclaration(None=OneMinus(None=Fresnel(ExponentIn=@n88, Normal=@n89) {baseReflectFraction=0})) {name=FresnelMask, nodeColor=<rgba 1 0.0469 0 1>}
  MaterialExpressionMultiply_8 = Multiply(A=TextureCoordinate(), B=ScalarParameter() {defaultValue=6, group=Diamond_RayMarching, parameterName=RayMarchingTilling})
  MaterialExpressionTransform_4 = Transform(None=CameraVectorWS()) {transformSourceType=TRANSFORMSOURCE_World, transformType=TRANSFORM_Tangent}
  n90 = Constant()
  MaterialExpressionMaterialFunctionCall_1 = MaterialFunctionCall(
   "Xb1,PS4"=Constant() {desc="This low step will brings some artifact around edges, but it is not that noticable", r=1}
   "PC-High&&Epic,XSX,PS5"=Constant() {r=1}
   "PC-Perf,Android"=@n90
   Switch=@n90
   "PC-Low(NotPerf)&&Medium"=Constant() {r=1}
  ) {materialFunction=&PlatformSwitch_1D}
  n91 = ScalarParameter() {defaultValue=0.3, group=Diamond_RayMarching, parameterName=DiamondRayMarchStepSize, sliderMax=0.2, sortPriority=13}
  n92 = ScalarParameter() {defaultValue=0.1, group=Diamond_RayMarching, parameterName=DiamondRayMarchStepSize_Perf, sliderMax=0.2, sortPriority=13}
  MaterialExpressionMaterialFunctionCall_6 = MaterialFunctionCall("Xb1,PS4"=@n91, "PC-High&&Epic,XSX,PS5"=@n91, "PC-Perf,Android"=@n92, Switch=@n92, "PC-Low(NotPerf)&&Medium"=@n91) {materialFunction=&PlatformSwitch_1D}
  MaterialExpressionTextureObjectParameter_1 = TextureObjectParameter() {
   group=Diamond_RayMarching
   parameterName=DiamondHeight
   samplerType=SAMPLERTYPE_Masks
   texture=&/T_BerryTart_DiamondHeight
  }
  MaterialExpressionNamedRerouteUsage_11 = NamedRerouteUsage()
  n93 = Custom!(
   uv=@MaterialExpressionMultiply_8
   viewDir=@MaterialExpressionTransform_4
   steps=@MaterialExpressionMaterialFunctionCall_1.Result
   stepdist=@MaterialExpressionMaterialFunctionCall_6.Result
   tex=@MaterialExpressionTextureObjectParameter_1
   Ntex=@MaterialExpressionNamedRerouteUsage_11
  ) {
   code="float3 rayStep = viewDir * -1;\r\nfloat4 heightTex = Texture2DSample(tex,texSampler,uv);\r\nfloat4 NormalTex;\r\n\r\n\r\nint Steps = floor(steps);\r\n//float steps = 16;\r\n//float stepdist = float(0.1);\r\n\r\n//[UNROLL]\r\nfor(int i=0; i <= Steps; i++)\r\n{\r\n\tif( max(heightTex.r, 0.0001) <= i  / Steps)\r\n\t{\r\n\t\treturn (NormalTex*i)/Steps; \r\n\t\t\r\n\t}\r\n\r\n\tfloat stepct =  heightTex.r*stepdist;\r\n\t\r\n\tuv += rayStep * stepct;\r\n    if( i == Steps)\r\n    {\r\n        NormalTex = Texture2DSample(Ntex,texSampler,uv.xy);\r\n    }\r\n    else\r\n    {\r\n\t    heightTex = Texture2DSample(tex,texSampler,uv.xy );\r\n        NormalTex = Texture2DSample(Ntex,texSampler,uv.xy);\r\n    }\r\n\t//RayMarchUV = uv.xy + distortStr;\r\n\r\n}\r\n\r\nreturn NormalTex;"
   showCode=true
   inputs [
    {inputName=uv, input {expression=&self:MaterialExpressionMultiply_8}}
    {inputName=viewDir, input {expression=&self:MaterialExpressionTransform_4}}
    {
     inputName=steps
     input {expression=&self:MaterialExpressionMaterialFunctionCall_1}
    }
    {
     inputName=stepdist
     input {expression=&self:MaterialExpressionMaterialFunctionCall_6}
    }
    {
     inputName=tex
     input {expression=&self:MaterialExpressionTextureObjectParameter_1}
    }
    {
     inputName=Ntex
     input {expression=&self:MaterialExpressionNamedRerouteUsage_11}
    }
   ]
  }
  n94 = TextureSample(UVs=@MaterialExpressionMultiply_8, Tex=NamedRerouteUsage())
  NamedRerouteDeclaration(None=MaterialFunctionCall!("Xb1,PS4"=@n93, "PC-High&&Epic,XSX,PS5"=@n93, "PC-Perf,Android"=@n94.RGB, Switch=@n94.RGB, "PC-Low(NotPerf)&&Medium"=@n93) {materialFunction=/Game/Effects/Niagara/MasterMaterials/Functions/Common/PlatformSwitch_3D}.Result) {name=RaymarchDepth, nodeColor=<rgba 0.6641 0 1 1>}
  NamedRerouteUsage()
  SkyAtmosphereLightDirection()
  Transform(None=NamedRerouteUsage())
 ]}
